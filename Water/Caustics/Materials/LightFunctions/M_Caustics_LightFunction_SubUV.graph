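# Light function material that projects animated water caustics from a flipbook texture onto lit surfaces below a water level.
# World XY is shifted along the MPC_Water sun vector by depth below water, scaled by Parallax, then tiled.
# Time at a playback FPS selects the SubUV cell.
# Depth blends from inverse Caustics Multiplier to the sampled texture, which defaults to a black placeholder; a debug value readout sits separately.
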Material /Water/Caustics/Materials/LightFunctions/M_Caustics_LightFunction_SubUV
graph {
 properties {materialDomain=MD_LightFunction, pixelDepthOffsetMode=PDOM_Legacy, usedShadingModels=Default Lit}
 nodes [
  n1 = ScalarParameter() {defaultValue=1, parameterName=Parallax}
  n2 = Normalize(VectorInput=ComponentMask(None=CollectionParameter() {collection=/Water/Materials/MPC/MPC_Water, parameterName=SunLightVector}) {b=true, g=true, r=true})
  n3 = Divide(A=ComponentMask(None=@n2) {g=true, r=true}, B=Abs(None=ComponentMask(None=@n2) {b=true}))
  n4 = Subtract(A=ScalarParameter() {parameterName=Water Level}, B=ComponentMask(None=WorldPosition().XYZ) {b=true})
  n5 = Subtract(A=ComponentMask(None=WorldPosition().XYZ) {g=true, r=true}, B=Multiply(A=Multiply(A=@n1, B=@n3), B=@n4))
  n6 = ScalarParameter() {defaultValue=2048, parameterName=Caustics Scale}
  n7 = Frac(None=Add(A=Divide(A=@n5, B=@n6)) {constB=0.5})
  n8 = ComponentMask(None=VectorParameter() {parameterName=SubUVFrames, defaultValue=<rgba 8 8 0 0>}.RGB) {g=true, r=true}
  n9 = ScalarParameter() {defaultValue=15, parameterName=Caustics Playback FPS}
  n10 = VectorParameter() {parameterName=SubUVFrames, defaultValue=<rgba 8 8 0 0>}
  n11 = MaterialFunctionCall!(
   1D Index=Fmod(
    A=MaterialFunctionCall(Speed=@n9) {materialFunction=&TimeWithSpeedVariable}.Time
    B=Multiply(A=@n10.R, B=@n10.R)
   )
   Cells XY=@n10.RGB
  ) {materialFunction=/Engine/Functions/Engine_MaterialFunctions01/Coordinates/1Dto2DIndex}
  n12 = ComponentMask(None=@n10.RGB) {g=true, r=true}
  n13 = TextureSampleParameter2D(UVs=Add(A=Divide(A=@n7, B=@n8), B=Divide(A=@n11.2D Index, B=@n12))) {
   parameterName=Caustics
   samplerType=SAMPLERTYPE_LinearColor
   texture=/Engine/Functions/Engine_MaterialFunctions02/PivotPainter2/Black_1x1_EXR_Texture
  }
  n14 = ScalarParameter() {defaultValue=128, parameterName=Fade Length}
  LinearInterpolate(A=Divide(B=ScalarParameter() {defaultValue=1, parameterName=Caustics Multiplier}), B=@n13.RGB, Alpha=Saturate(None=Divide(A=@n4, B=@n14))) {constA=1}
  n15 = ComponentMask(None=WorldPosition().XYZ) {g=true, r=true}
  MaterialFunctionCall!(
   Vector3=ComponentMask(None=CollectionParameter() {collection=&MPC_Water}) {b=true, g=true, r=true}
   MaximumNumberOfDigits
   UVs=Frac(None=Divide(A=@n15) {constB=256})
   DebugTextLocation RG_UpperRight BA_LowerLeft
   Component Spacing
  ) {materialFunction=/Engine/Functions/Engine_MaterialFunctions02/Utility/DebugFloat3Values}
 ]
}
