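# Sun light function material that builds water caustics, though the graph shows no branch wired to a material output.
# World XY is shifted along the MPC_Water sun vector by depth below water level to sample caustics, with a depth fade.
# A custom "Volume Blur" loop smears caustics along combined editor-camera and sun directions, using temporally dithered step counts.
# Separate lerps form faded and blur-mixed results, alongside a camera-position debug readout.
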Material /Water/Caustics/Materials/LightFunctions/M_Caustics_LightFunction_Sun
graph {
 properties {materialDomain=MD_LightFunction, pixelDepthOffsetMode=PDOM_Legacy, usedShadingModels=Default Lit}
 nodes [
  CollectionParameter() {collection=/Water/Materials/MPC/MPC_Water}
  n1 = ComponentMask(None=WorldPosition().XYZ) {g=true, r=true}
  n2 = Normalize(VectorInput=ComponentMask(None=CollectionParameter() {collection=&MPC_Water, parameterName=SunLightVector}) {b=true, g=true, r=true})
  n3 = ComponentMask(None=@n2) {g=true, r=true}
  n4 = Abs(None=ComponentMask(None=@n2) {b=true})
  n5 = Subtract(A=ScalarParameter() {parameterName=Water Level}, B=ComponentMask(None=WorldPosition().XYZ) {b=true})
  n6 = Multiply(A=Multiply(A=ScalarParameter() {defaultValue=1, parameterName=Parallax}, B=Divide(A=@n3, B=@n4)), B=@n5)
  MaterialExpressionAdd_0 = Add(A=Divide(A=Subtract(A=@n1, B=@n6), B=ScalarParameter() {defaultValue=2048, parameterName=Caustics Scale})) {constB=0.5}
  n7 = TextureSampleParameter2D(UVs=@MaterialExpressionAdd_0) {
   parameterName=Caustics
   samplerType=SAMPLERTYPE_LinearColor
   texture=/Engine/Functions/Engine_MaterialFunctions02/PivotPainter2/Black_1x1_EXR_Texture
  }
  n8 = ScalarParameter() {defaultValue=128, parameterName=Fade Length}
  LinearInterpolate(A=Divide(B=ScalarParameter() {defaultValue=1, parameterName=Caustics Multiplier}), B=@n7.RGB, Alpha=Saturate(None=Divide(A=@n5, B=@n8))) {constA=1}
  MaterialExpressionTextureObjectParameter_1 = TextureObjectParameter() {parameterName=Caustics, samplerType=SAMPLERTYPE_LinearColor, texture=/Water/Textures/Caustics/Caustics_Animated_01_HDR}
  n9 = CollectionParameter() {collection=&MPC_Water, parameterName=EditorCameraLocation}
  n10 = Normalize(VectorInput=Subtract(A=ComponentMask(None=@n9) {b=true, g=true, r=true}, B=WorldPosition().XYZ))
  n11 = ComponentMask(None=@n10) {g=true, r=true}
  n12 = Abs(None=ComponentMask(None=@n10) {b=true})
  n13 = Normalize(VectorInput=ComponentMask(None=CollectionParameter() {collection=&MPC_Water, parameterName=SunLightVector}) {b=true, g=true, r=true})
  MaterialExpressionAdd_3 = Add(
   A=Multiply(A=Divide(A=@n11, B=@n12), B=ScalarParameter() {defaultValue=0.5, parameterName=RayDirMult}) {constB=0.5}
   B=Multiply(A=ComponentMask(None=@n13) {g=true, r=true}) {constB=0.5}
  )
  MaterialExpressionMultiply_5 = Multiply(
   A=Divide(A=@n5, B=ScalarParameter() {defaultValue=2048, parameterName=Caustics Scale})
   B=ScalarParameter() {defaultValue=1, parameterName=Blur Distance}
  )
  n14 = ViewProperty() {property=MEVP_TemporalSampleIndex}
  MaterialExpressionAdd_2 = Add(A=Multiply(A=TextureCoordinate(), B=Constant() {r=1000}), B=AppendVector(A=@n14.Property, B=@n14.Property))
  n15 = Custom!(p=@MaterialExpressionAdd_2) {
   code="Mod( ((uint)(p.x) + 2 * (uint)(p.y)) , 5 )"
   outputType=CMOT_Float1
   inputs [{inputName=p, input {expression=&self:MaterialExpressionAdd_2}}]
  }
  n16 = TextureSample(UVs=Multiply(A=TextureCoordinate(), B=Constant2Vector() {g=32, r=32})) {samplerType=SAMPLERTYPE_LinearGrayscale, texture=/Engine/EngineMaterials/Good64x64TilingNoiseHighFreq}
  n17 = Add(A=Divide(A=Add(A=@n15, B=@n16.R) {constB=0}) {constB=6}) {constB=-0.5}
  MaterialExpressionMultiply_1 = Multiply(
   A=ScalarParameter() {defaultValue=8, parameterName=Blur Steps}
   B=LinearInterpolate(A=ScalarParameter() {defaultValue=3, parameterName=Dither Reduction}, Alpha=Saturate(None=@n17)) {constA=0.5}
  )
  MaterialExpressionScalarParameter_6 = ScalarParameter() {parameterName=Caustics Volume Blur Mip}
  n18 = Custom!(
   Tex=@MaterialExpressionTextureObjectParameter_1
   UVW=@MaterialExpressionAdd_0
   Vector=@MaterialExpressionAdd_3
   Distance=@MaterialExpressionMultiply_5
   Steps=@MaterialExpressionMultiply_1
   Mip=@MaterialExpressionScalarParameter_6
  ) {
   code="\r\n\r\nfloat4 accum = 0;\r\nfloat2 curUVW = UVW;\r\n\r\nfloat stepsize = Distance / Steps;\r\n\r\nfloat gradstep = 1.0 / Steps;\r\nfloat gradpos = 0;\r\n\r\nfor(int i =0; i < Steps; i++)\r\n{\r\n\r\n\r\n\tgradpos = float(i) * gradstep;\r\n\t//float3 spectrum = saturate(1-abs(2*gradpos- float3(0,1,2)));\r\n\t\r\n\tfloat3 caustics = Tex.SampleLevel(TexSampler, curUVW, Mip).rgb;\r\n\taccum.rgb += caustics;\r\n\t//accum.a += caustics.r * spectrum.b;\r\n\t\r\n\tcurUVW += Vector.xy * stepsize;\r\n}\r\n\r\nreturn float4(accum/Steps );"
   description=Volume Blur
   outputType=CMOT_Float4
   inputs [
    {
     inputName=Tex
     input {expression=&self:MaterialExpressionTextureObjectParameter_1}
    }
    {inputName=UVW, input {expression=&self:MaterialExpressionAdd_0}}
    {inputName=Vector, input {expression=&self:MaterialExpressionAdd_3}}
    {inputName=Distance, input {expression=&self:MaterialExpressionMultiply_5}}
    {inputName=Steps, input {expression=&self:MaterialExpressionMultiply_1}}
    {inputName=Mip, input {expression=&self:MaterialExpressionScalarParameter_6}}
   ]
  }
  n19 = Multiply(A=ComponentMask(None=@n18) {b=true, g=true, r=true}, B=ScalarParameter() {defaultValue=0.5, parameterName=Volume Blur})
  LinearInterpolate(A=Add(A=@n19, B=@n7.RGB), B=@n19, Alpha=ScalarParameter() {parameterName=Blur only, sliderMin=1})
  n20 = ComponentMask(None=WorldPosition().XYZ) {g=true, r=true}
  MaterialFunctionCall!(
   Vector3=ComponentMask(None=CameraPositionWS()) {b=true, g=true, r=true}
   MaximumNumberOfDigits
   UVs=Frac(None=Divide(A=@n20) {constB=256})
   DebugTextLocation RG_UpperRight BA_LowerLeft
   Component Spacing
  ) {materialFunction=/Engine/Functions/Engine_MaterialFunctions02/Utility/DebugFloat3Values}
 ]
}
