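# Light function material named lf-05, stored among the water caustics light functions.
# Its graph holds only a scalar parameter, FF, with no texture or math nodes and no connection to a material output shown.
# As written, it contains no caustic pattern logic of its own.
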Material /Water/Caustics/Materials/LightFunctions/lf-05
graph {
 properties {materialDomain=MD_LightFunction, pixelDepthOffsetMode=PDOM_Legacy, usedShadingModels=Default Lit}
 nodes [ScalarParameter() {defaultValue=0.5, parameterName=FF}]
}
